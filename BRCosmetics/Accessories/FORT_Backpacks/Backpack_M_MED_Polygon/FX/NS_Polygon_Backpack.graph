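# Niagara effect for the Polygon backpack cosmetic, shedding small, pink-red mesh shards that flare bright and fade.
# Two near-identical emitters spawn from a user skeletal mesh: a GPU one samples it directly, a CPU one adds a sphere location.
# Shards orient to sampled or sphere normals, with a dot-product test culling by facing.
# Directional push, curl noise, drag and spin move them, while inherited source movement tapers off through a curve.
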NiagaraSystem /BRCosmetics/Accessories/FORT_Backpacks/Backpack_M_MED_Polygon/FX/NS_Polygon_Backpack
dependencies {usedDataInterfaces [&NiagaraDataInterfaceCurve], usedDynamicInputs [], usedModules [&SystemState]}
emitters {
 Empty {
  simTarget=GPUComputeSim
  emitterUpdateScript [
   &EmitterState()
   &SpawnRate(SpawnRate=100, Use Spawn Probability=false)
   &SpawnBurst_Instantaneous(Spawn Count=200, Use Spawn Probability=false, UseLoopCountLimit=false)
  ]
  particleSpawnScript [
   &InitializeParticle(
    Color=<rgba 1 0.117 0.1845 1>
    Lifetime Max=1.25
    Lifetime Min=0.25
    Lifetime Mode=::NewEnumerator1
    Mesh Scale Mode=::NewEnumerator1
    Mesh Uniform Scale Max=0.5
    Mesh Uniform Scale Min=0.2
    UsePositionOffset=false
    Write Material Random=true
   )
   &SkeletalMeshLocation(
    MaskSpawnGroup=false
    Mesh Sampling Type=::NewEnumerator7
    OffsetAlongNormal=false
    OverrideRandomSeed=false
    Skeletal Mesh=$User.New Skeletal Mesh
   )
   &InitialMeshOrientation(
    Orientation Axis=<0 0 1>
    Orientation Vector=$Particles.SkeletalMeshLocation.SampledNormal
    Use Orientation Vector=true
    Use Rotation Vector=false
   )
   &KillParticles(Kill Particles=&SetBoolByFloatComparison(
     A=&DotProduct(
      Vector A=$Particles.SkeletalMeshLocation.SampledNormal
      Vector B=&TransformVector(Vector=<0 1 0>, Destination Space=::Simulation)
     )
     B=0.5
    ))
   &CameraOffset(Camera Offset Amount=&RandomRangeFloat(Minimum=2, Maximum=5))
  ]
  particleUpdateScript [
   &ParticleState()
   &PointForce(
    Force Strength=75
    Linear Force Vector:Vector3f=&TransformVector(Vector=<0 -1 0>, Destination Space=::Simulation)
    Module.Use Falloff=false
    Module.UseLinearForce=true
   )
   &CurlNoiseForce(Noise Strength=250, Pan Noise=true, Pan Noise Field=<0 0 -10>)
   &Drag(Drag=&RandomRangeFloat(Minimum=2, Maximum=5), Use Linear Drag=true, Use Rotational Drag=false)
   &InheritSourceMovement(
    Applied Position Scale=&VectorFromFloat(Value=&FloatFromCurve(FloatCurve=DataInterface() {propertyValues {Curve {keys [{tangentMode=RCTM_User, time=0, value=1}, {tangentMode=RCTM_User, time=1, value=0}]}}}))
    ApplyForce=false
    ApplyPosition=true
    ApplyPositionFalloff=false
   )
   &SolveForcesAndVelocity(Clamp Velocity=false, Limit Acceleration=false)
   &MeshRotationForce(Rotation=<10 10 10>)
   &SolveRotationalForcesAndVelocity()
   &ScaleColor(
    Scale RGB=&VectorFromFloat(Value=&FloatFromCurve(FloatCurve=DataInterface() {propertyValues {Curve {keys [
          <at 0 10 RCIM_Cubic>
          {arriveTangent=-4.0325, interpMode=RCIM_Cubic, leaveTangent=-4.0325, tangentMode=RCTM_User, time=0.4, value=1}
          <at 1 0 RCIM_Cubic>
         ]}}}))
    ScaleA=true
    ScaleRGB=true
   )
  ]
  rendererClasses [&NiagaraMeshRendererProperties]
  renderers [{rendererClass=&NiagaraMeshRendererProperties, rendererIndex=0}]
 }
 Empty001 {
  emitterUpdateScript [
   &EmitterState()
   &SpawnRate(SpawnRate=100, Use Spawn Probability=false)
   &SpawnBurst_Instantaneous(Spawn Count=200, Use Spawn Probability=false, UseLoopCountLimit=false)
  ]
  particleSpawnScript [
   &InitializeParticle(
    Color=<rgba 1 0.117 0.1845 1>
    Lifetime Max=1.25
    Lifetime Min=0.25
    Lifetime Mode=::NewEnumerator1
    Mesh Scale Mode=::NewEnumerator1
    Mesh Uniform Scale Max=0.5
    Mesh Uniform Scale Min=0.2
    UsePositionOffset=false
    Write Material Random=true
   )
   &SkeletalMeshLocation(MaskSpawnGroup=false, OverrideRandomSeed=false, Position Sampling=::NewEnumerator1, Skeletal Mesh=$User.New Skeletal Mesh)
   &SphereLocation(
    MaskSpawnGroup=false
    Override Seed=false
    Sphere Origin:NiagaraPosition=/Niagara/DynamicInputs/Helpers/SimulationPosition()
    Sphere Radius=25
    Surface Only=false
    UseNonUniformScale=false
    UseOffset=false
    UseSphereOrientation=false
   )
   &InitialMeshOrientation(Orientation Axis=<0 0 1>, Orientation Vector=$Output.SphereLocation.SphereNormal, Use Orientation Vector=true, Use Rotation Vector=false)
   &KillParticles(Kill Particles=&SetBoolByFloatComparison(
     A=&DotProduct(
      Vector A=$Output.SphereLocation.SphereNormal
      Vector B=&TransformVector(Vector=<0 1 0>, Destination Space=::Simulation)
     )
     B=0.5
    ))
   &CameraOffset(Camera Offset Amount=&RandomRangeFloat(Minimum=2, Maximum=5))
  ]
  particleUpdateScript [
   &ParticleState()
   &PointForce(
    Force Strength=75
    Linear Force Vector:Vector3f=&TransformVector(Vector=<0 -1 0>, Destination Space=::Simulation)
    Module.Use Falloff=false
    Module.UseLinearForce=true
   )
   &CurlNoiseForce(Noise Strength=250, Pan Noise=true, Pan Noise Field=<0 0 -10>)
   &Drag(Drag=&RandomRangeFloat(Minimum=2, Maximum=5), Use Linear Drag=true, Use Rotational Drag=false)
   &InheritSourceMovement(
    Applied Position Scale=&VectorFromFloat(Value=&FloatFromCurve(FloatCurve=DataInterface() {propertyValues {Curve {keys [{tangentMode=RCTM_User, time=0, value=1}, {tangentMode=RCTM_User, time=1, value=0}]}}}))
    ApplyForce=false
    ApplyPosition=true
    ApplyPositionFalloff=false
   )
   &SolveForcesAndVelocity(Clamp Velocity=false, Limit Acceleration=false)
   &MeshRotationForce(Rotation=<10 10 10>)
   &SolveRotationalForcesAndVelocity()
   &ScaleColor(
    Scale RGB=&VectorFromFloat(Value=&FloatFromCurve(FloatCurve=DataInterface() {propertyValues {Curve {keys [
          <at 0 10 RCIM_Cubic>
          {arriveTangent=-4.0325, interpMode=RCIM_Cubic, leaveTangent=-4.0325, tangentMode=RCTM_User, time=0.4, value=1}
          <at 1 0 RCIM_Cubic>
         ]}}}))
    ScaleA=true
    ScaleRGB=true
   )
  ]
  rendererClasses [&NiagaraMeshRendererProperties]
  renderers [{rendererClass=&NiagaraMeshRendererProperties, rendererIndex=0}]
 }
}
summary {
 systemName=NS_Polygon_Backpack
 emitters [
  {emitterName=Empty, simTarget=GPUComputeSim, rendererClasses [&NiagaraMeshRendererProperties]}
  {emitterName=Empty001, rendererClasses [&NiagaraMeshRendererProperties]}
 ]
 userVariables [User.New Skeletal Mesh {
   type=NiagaraDataInterfaceSkeletalMesh
   defaultValue {
    struct=NiagaraExt_VariableValue_DataInterface
    value {dataInterface=&self:NiagaraDataInterfaceSkeletalMesh_1}
   }
  }]
}
