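# Physics asset for the M_MED_Polygon backpack cosmetic, defining secondary-motion bodies and joints for its dangling parts.
# The pack body combines three capsules and anchors single-capsule bodies for a cartridge, left and right swords and six fur pieces.
# A three-link tassle chain also hangs from the pack, each link constrained to the one before it.
# Each joint carries swing and twist angle limits tuned by part type.
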PhysicsAsset /BRCosmetics/Accessories/FORT_Backpacks/Backpack_M_MED_Polygon/Meshes/M_MED_Polygon_Pack_Physics
bodies {
 dyn_cartridge_1 [{shapeType=Capsule, length=6.7303, radius=4.0877, shapeName=dyn_cartridge_1_capsule, center=<0 -0.2657 0>, rotation=<rot 90 0 0>}]
 dyn_fur_1 [{shapeType=Capsule, length=0.505, radius=0.505, shapeName=dyn_fur_1_capsule, center=<3.7106 0 0>}]
 dyn_fur_2 [{shapeType=Capsule, length=0.505, radius=0.505, shapeName=dyn_fur_2_capsule, center=<3.7106 0 0>}]
 dyn_fur_3 [{shapeType=Capsule, length=0.505, radius=0.505, shapeName=dyn_fur_3_capsule, center=<3.7106 0 0>}]
 dyn_fur_4 [{shapeType=Capsule, length=0.505, radius=0.505, shapeName=dyn_fur_4_capsule, center=<3.7106 0 0>}]
 dyn_fur_5 [{shapeType=Capsule, length=0.505, radius=0.505, shapeName=dyn_fur_5_capsule, center=<3.7106 0 0>}]
 dyn_fur_6 [{shapeType=Capsule, length=0.505, radius=0.505, shapeName=dyn_fur_6_capsule, center=<3.7106 0 0>}]
 dyn_sword_l_1 [{shapeType=Capsule, length=21.6766, radius=3.7094, shapeName=dyn_sword_l_1_capsule, center=<1.2193 0 0>, rotation=<rot 90 0 0>}]
 dyn_sword_r_1 [{shapeType=Capsule, length=21.6714, radius=3.2005, shapeName=dyn_sword_r_1_capsule, center=<1.2193 0 0>, rotation=<rot 90 -0.0011 -0.001>}]
 dyn_tassle_1 [{shapeType=Capsule, length=1.4964, radius=0.5, shapeName=dyn_tassle_1_capsule}]
 dyn_tassle_2 [{shapeType=Capsule, length=1.0706, radius=0.6102, shapeName=dyn_tassle_2_capsule}]
 dyn_tassle_3 [{shapeType=Capsule, length=1.5435, radius=0.5145, shapeName=dyn_tassle_3_capsule}]
 pack [
  {shapeType=Capsule, length=1, radius=1, shapeName=pack_capsule}
  {shapeType=Capsule, length=6.4703, radius=7.486, shapeName=pack_capsule, center=<4.963 -1.0685 -0.9918>, rotation=<rot -26.6847 0 0>}
  {shapeType=Capsule, length=3.75, radius=2.5, shapeName=pack_capsule, center=<-16.9177 -6.7189 7.898>, rotation=<rot -20 0 -10.0005>}
 ]
}
constraints [
 {bone1Name=dyn_tassle_2, bone2Name=dyn_tassle_1, swing1LimitDegrees=45, swing2LimitDegrees=10, twistLimitDegrees=10}
 {bone1Name=dyn_tassle_3, bone2Name=dyn_tassle_2, swing1LimitDegrees=45, swing2LimitDegrees=10, twistLimitDegrees=10}
 {bone1Name=dyn_sword_l_1, bone2Name=pack, swing1LimitDegrees=50, swing2LimitDegrees=30, twistLimitDegrees=30}
 {bone1Name=dyn_sword_r_1, bone2Name=pack, swing1LimitDegrees=50, swing2LimitDegrees=30, twistLimitDegrees=30}
 {bone1Name=dyn_fur_1, bone2Name=pack, swing1LimitDegrees=30, swing2LimitDegrees=30, twistLimitDegrees=10}
 {bone1Name=dyn_fur_2, bone2Name=pack, swing1LimitDegrees=30, swing2LimitDegrees=30, twistLimitDegrees=10}
 {bone1Name=dyn_fur_3, bone2Name=pack, swing1LimitDegrees=30, swing2LimitDegrees=30, twistLimitDegrees=10}
 {bone1Name=dyn_fur_4, bone2Name=pack, swing1LimitDegrees=30, swing2LimitDegrees=30, twistLimitDegrees=10}
 {bone1Name=dyn_fur_5, bone2Name=pack, swing1LimitDegrees=30, swing2LimitDegrees=30, twistLimitDegrees=10}
 {bone1Name=dyn_fur_6, bone2Name=pack, swing1LimitDegrees=30, swing2LimitDegrees=30, twistLimitDegrees=10}
 {bone1Name=dyn_cartridge_1, bone2Name=pack, swing1LimitDegrees=45, swing2LimitDegrees=45, twistLimitDegrees=10}
 {bone1Name=dyn_tassle_1, bone2Name=pack, swing1LimitDegrees=45, swing2LimitDegrees=12.4286, twistLimitDegrees=0}
]
